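# M_SkeleMeshGameplayActor is a minimal Default Lit material, named for skeletal mesh gameplay actors in the DS_Fortnight package.
# Its graph holds a single orange constant colour node, with no link to a material output shown.
# Usage flags enable it for skeletal meshes and mesh deformers.
# It has no textures, parameters or other shading logic.
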Material /Clyde/Packages/DS_Fortnight/SM/Materials/M_SkeleMeshGameplayActor
graph {
 properties {bUsedWithMeshDeformer=true, bUsedWithSkeletalMesh=true, pixelDepthOffsetMode=PDOM_Legacy, usedShadingModels=Default Lit}
 nodes [Constant3Vector() {constant=<rgba 1 0.3392 0 0>}]
}
